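# Animation Blueprint for the SkelFlask backpack that adjusts how the pack sits on the character.
# The AnimGraph layers four bone-space modifiers on the pack bone over the input pose, with additive offsets and blendable alphas.
# Each update maps spine bend to pack roll and shift, derives a melee-gated arm alpha, and flags body meshes named SkelFlask.
# Customization load caches the pawn, body mesh, gender and large-body flag.
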AnimBlueprint /BRCosmetics/Accessories/FORT_Backpacks/Backpack_M_MED_SkelFlask/Meshes/M_MED_SkelFlask_Pack_AnimBP
target_skeleton=&/M_MED_SkelFlask_Pack_Skeleton
graphs [
 AnimGraph {nodes [
   n1 = "Transform (Modify) Bone - Bone: pack"!(ComponentPose=Local To Component(LocalPose=Input Pose()), Alpha, Translation="0,3,0") -> (Pose) {
    translationMode=BMM_Additive
    translationSpace=BCS_BoneSpace
    alphaBoolBlend {blendOption=Linear}
    boneToModify {boneName=pack}
    translation {y=3}
   }
   n2 = "Transform (Modify) Bone - Bone: pack"!(ComponentPose=@n1, Alpha, Translation="0,0,0", Rotation:Rotator="0,0,0") -> (Pose) {
    rotationMode=BMM_Additive
    translationMode=BMM_Additive
    translationSpace=BCS_BoneSpace
    alphaBoolBlend {blendOption=Linear}
    boneToModify {boneName=pack}
   }
   n3 = "Transform (Modify) Bone - Bone: pack"!(ComponentPose=@n2, bAlphaBoolEnabled:bool, Translation="0,-5,0", Rotation="0,0,0", Scale:Vector) -> (Pose) {
    alphaInputType=Bool
    rotationMode=BMM_Additive
    translationMode=BMM_Additive
    translationSpace=BCS_BoneSpace
    alphaBoolBlend {blendInTime=0.25, blendOption=Linear, blendOutTime=0.25}
    boneToModify {boneName=pack}
    translation {y=-5}
   }
   n4 = "Transform (Modify) Bone - Bone: pack"!(ComponentPose=@n3, Alpha:float, Translation="0,-3,0", Rotation="0,-35,0", Scale) -> (Pose) {
    rotationMode=BMM_Additive
    translationMode=BMM_Additive
    translationSpace=BCS_BoneSpace
    alphaScaleBiasClamp {bInterpResult=true}
    boneToModify {boneName=pack}
    rotation {yaw=-35}
    translation {y=-3}
   }
   Output Pose(Result=Component To Local(ComponentPose=@n4)) {layerGroup=None}
  ]}
 BackPack_Adjustments {nodes [
   n1 = BackPack_Adjustments!() -> (
    then
    In Range A Rotate:double=~
    In Range B Rotate:double=~
    Out Range A Rotate:double=~
    Out Range B Rotate:double=~
    In Range A Translate:double=~
    In Range B Translate:double=~
    Out Range A Translate:double=~
    Out Range B Translate:double=~
   ) {is=title K2Node_FunctionEntry}
   n2 = Get Skel_Mesh()
   n3 = BreakRotator(InRot=BreakTransform(InTransform=GetDeltaTransformFromRefPose(self=@n2, BoneName=spine_05, BaseName=spine_01)).Rotation)
   > Set backpack_rotate(backpack_rotate=MakeRotator(Roll=MapRangeClamped(Value=@n3.Yaw, InRangeA=@n1, InRangeB=@n1.In Range B Rotate, OutRangeA=@n1.Out Range A Rotate, OutRangeB=@n1.Out Range B Rotate)))
   > Set backpack_translation(backpack_translation=MakeVector(Y=MapRangeClamped(
      Value=@n3.Yaw
      InRangeA=@n1.In Range A Translate
      InRangeB=@n1.In Range B Translate
      OutRangeA=@n1.Out Range A Translate
      OutRangeB=@n1.Out Range B Translate
     )))
  ]}
 EventGraph {nodes [
   Event OnCharacterCustomizationLoaded()
   > n1 = Set Owning Pawn(Owning Pawn:Pawn Object Reference=TryGetFortPawn()) -> (Output_Get:Pawn Object Reference)
   > Is Valid(InputObject=@n1)
   > n2 = Sequence!(execute) -> (then_0, then_1, then_2:Exec)
   n3 = FindGender!(execute=@n2.then_1, Object:Object Reference=Get Owning Pawn() -> (Owning Pawn:Pawn Object Reference)) -> (True:Exec, False:Exec, OutGender:EFortCustomGender Enum=Invalid)
   > Set isMale(isMale:bool="Equal (Enum)"(A:EFortCustomGender Enum=@n3, B:EFortCustomGender Enum=Male)) -> (Output_Get:bool)
   > Set isFemale(isFemale:bool="Equal (Enum)"(A=@n3, B=Female)) -> (Output_Get:bool)
   n4 = FindBodyType!(execute=@n2.then_2, Object:Object Reference=Get Owning Pawn()) -> (True:Exec, False:Exec, OutBodyType:EFortCustomBodyType Enum=NONE)
   > Set isLarge(isLarge:bool="Equal (Enum)"(A:EFortCustomBodyType Enum=@n4, B:EFortCustomBodyType Enum=Large)) -> (Output_Get:bool)
   n5 = Cast To FortPlayerPawn(execute=@n2.then_0, Object=Get Owning Pawn())
   > n6 = Set Owning Player Pawn(Owning Player Pawn:Fort Player Pawn Object Reference=@n5) -> (Output_Get:Fort Player Pawn Object Reference)
   > Set Body Mesh(Body Mesh:Skeletal Mesh Component Object Reference=GetSkeletalMeshForPartType(self=@n6)) -> (Output_Get:Skeletal Mesh Component Object Reference)
   Event BlueprintUpdateAnimation()
   n7 = Get Skel_Mesh()
   > Set ArmAlpha(ArmAlpha:double=float * float(
     A=GetAlphaFromReferenceTransformDelta(SkeletalMeshComponent=@n7, BoneName=upperarm_l, BaseName=lowerarm_l, InputRange=(X=-40,Y=-140), TwistAxis="1,0,0")
     B:bool=Get bIsPlayingMeleeAnim()
    ) {is=title K2Node_PromotableOperator}) -> (Output_Get:double)
   > BackPack_Adjustments!(
    execute
    self
    In Range A Rotate:double=0
    In Range B Rotate:double=-20
    Out Range A Rotate:double=0
    Out Range B Rotate:double=15
    In Range A Translate:double=0
    In Range B Translate:double=2.5
    Out Range A Translate:double=0
    Out Range B Translate:double=-1
   ) -> (then) {is=title K2Node_CallFunction}
   > n8 = Cast To FortPlayerPawn(Object=TryGetPawnOwner())
   n9 = GetDisplayName(Object=GetSkeletalMeshForPartType(self=@n8))
   > Set isSkelFlaskAlpha(isSkelFlaskAlpha:double="To Float (Boolean)"(InBool=MatchesWildcard(SourceString=@n9, Wildcard=*SkelFlask*))) -> (Output_Get:double)
   Event BlueprintInitializeAnimation()
   > Set Skel_Mesh(Skel_Mesh=GetOwnerSkeletalMeshComponent())
  ]}
]
variables [
 ArmAlpha {type=number}
 backpack_rotate {type=Rotator}
 backpack_translation {type=Vector, default=<0 -1.4 0>}
 Body Mesh {type=SkeletalMeshComponent}
 Gravity Override Params {type=GravityOverrideParamsStruct}
 GravityOverride {type=Vector}
 is NOT M_Large {type=boolean, default=true}
 isFemale {type=boolean}
 isLarge {type=boolean}
 isMale {type=boolean}
 isMedium {type=boolean}
 isSkelFlaskAlpha {type=number}
 Owning Pawn {type=Pawn, default=null}
 Owning Player Pawn {type=FortPlayerPawn}
 Skel_Mesh {type=SkeletalMeshComponent}
 SkyDivingAlpha {type=number, default=1}
]
